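# Niagara idle glow effect for the VileSting character intro animation, built from one looping sprite emitter.
# Bursts three soft-positioned sprites per loop, each with its own size, tilt and offset by spawn index, tinted by a system TOD color.
# The sprites follow a sampled socket, and their width narrows or widens with the socket axis's angle to the camera.
# Alpha fades in, weighted per sprite and dimmed in the front end.
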NiagaraSystem /BRCosmetics/Animation/Game/MainPlayer/CharIntros/s40/VileSting/FX/NS_IntroAnim_VileSting_IdleGlow
dependencies {
 usedDataInterfaces [
  &NiagaraDataInterfaceCurve
  &NiagaraDataInterfaceCamera
  &NiagaraDataInterfaceColorCurve
  &NiagaraDataInterfaceSocketReader
 ]
 usedDynamicInputs [&ColorFromCurve]
 usedModules [
  &self:SystemSpawnScript.SystemScriptSource.SystemScriptGraph.NiagaraNodeAssignment_0.SetVariables_#1
  &self:SystemSpawnScript.SystemScriptSource.SystemScriptGraph.NiagaraNodeAssignment_1.SetVariables_#2
  &SystemState
  &self:SystemSpawnScript.SystemScriptSource.SystemScriptGraph.NiagaraNodeAssignment_2.SetVariables_#3
 ]
}
emitters {LightAndGlow {
  emitterUpdateScript [
   &EmitterState(Life Cycle Mode=::NewEnumerator1, Loop Behavior=::NewEnumerator1, Loop Duration=4.6, UseLoopDelay=false)
   &SpawnBurst_Instantaneous(Spawn Count=3, Use Spawn Probability=false, UseLoopCountLimit=false)
  ]
  particleSpawnScript [
   &InitializeParticle(
    Color=&MultiplyLinearColor(Linear Color=<rgba 1 1 1 0.5>, Scale Factor=&MakeVector4_FromLinearColor(LinearColor=$System.TOD))
    Lifetime=4.6
    Position Offset=&MergeHelper_VectorSwitch(0=<1.5 0.5 20>, 1=<3.5 0.5 30>, 2=<8 0.5 42>, 3=?, Index=&ReturnExecIndex(), Num=::NewEnumerator1)
    Sprite Size=&MergeHelper_Vector2Switch(0=<11 30>, 1=<13 50>, 2=<10 40>, 3=?, 4=?, Index=&ReturnExecIndex(), Num=::NewEnumerator1)
    Sprite Size Mode=::NewEnumerator4
    UsePositionOffset=true
    Write Material Random=true
   )
   &CameraOffset(Camera Offset Amount=-1)
   &SpriteFacingAndAlignment(
    Sprite Alignment=&MergeHelper_VectorSwitch(0=<0.1 0 1>, 1=<0.3 0 1>, 2=<0.7 0 1>, 3=?, Index=&ReturnExecIndex(), Num=::NewEnumerator1)
    WriteAlignment=true
    WriteFacing=true
   )
  ]
  particleUpdateScript [
   &ParticleState(Kill Particles When Lifetime Has Elapsed=false)
   &ScaleColor(
    Scale Alpha=&FloatFromCurve(
     FloatCurve=DataInterface() {propertyValues {Curve {keys [
         <at 0 0 RCIM_Cubic>
         {arriveTangent=0.03, interpMode=RCIM_Cubic, leaveTangent=0.03, tangentMode=RCTM_User, time=0.1, value=1}
         {arriveTangent=-0.0058, interpMode=RCIM_Cubic, leaveTangent=-0.0058, tangentMode=RCTM_User, time=0.9, value=1}
        ]}}}
     Scale Curve=&Multiply_Float(
      A=&MergeHelper_FloatSwitch(0=1.1, 1=0.8, 2=0.5, 3=?, Index=&ReturnExecIndex(), Num=::NewEnumerator1)
      B=&FN_FrontEndFloatModifier(Gameplay=1, Front End=0.5)
     )
    )
    ScaleA=true
    ScaleRGB=true
   )
   &SocketLocation(
    Orientation Sampling=::NewEnumerator1
    Position Sampling=::NewEnumerator1
    SamplePosOffset=true
    Sampled Position Offset=&MergeHelper_VectorSwitch(0=<0 0.5 20>, 1=<3 0.5 30>, 2=<8 0.5 42>, 3=?, Index=&ReturnExecIndex(), Num=::NewEnumerator1)
    Socket Index=2
    Socket Reader=$System.Socket Reader
    Socket Sampling Mode=::NewEnumerator3
    VelocitySpeedLimit=false
   )
   /SetVariables(Particles.SpriteSize=&Multiply_Vector2D(
     A=$Particles.Initial.SpriteSize
     B=&MakeVector2D(
      X=&Lerp_Float(
       A=2
       B=0.5
       Alpha=&Abs_Float(Float:NiagaraFloat=&DotProduct(
         Vector A=&RotateVectorByQuaternion(Vector To Rotate=<1 0 0>, Quaternion=$Output.SocketLocation.SampledRotation)
         Vector B=&ReturnCameraVector(Camera Vectors=::NewEnumerator4, Coordinate Space=::Simulation, Origin Position=$Particles.Position)
        ))
      )
      Y=1
     )
    )) {isSetParameters=true}
  ]
  rendererClasses [&NiagaraSpriteRendererProperties]
  renderers [{rendererClass=&NiagaraSpriteRendererProperties}]
 }}
summary {
 systemName=NS_IntroAnim_VileSting_IdleGlow
 emitters [{emitterName=LightAndGlow, rendererClasses [&NiagaraSpriteRendererProperties]}]
 userVariables [
  User.bIsFrontEnd {type=NiagaraBool}
  User.bIsFrontEndPreview {type=NiagaraBool}
  User.RibbonIDOffset {type=NiagaraInt32}
  User.Variant {type=NiagaraFloat, defaultValue=0, description=""}
 ]
}
